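# Niagara system for the Gaudy vehicle drift trail cosmetic, drawing trails and effects at two wheels.
# Two multi-renderer ribbons stream from the wheels with curve-driven color fades, while burst sprites scatter around wheel sockets.
# Persistent wheel-wrap meshes follow those sockets and are scaled per wheel.
# All three emitters resolve colors through painted-variant overrides and pass per-wheel indices to a tire-FX disable module.
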NiagaraSystem /VehicleCosmetics/DriftTrail/DriftTrail_Gaudy/FX/NS_DriftTrail_Gaudy_MP_
dependencies {
 usedDataInterfaces [&NiagaraDataInterfaceCurve, &NiagaraDataInterfaceArrayColor, &NiagaraDataInterfaceSkeletalMesh]
 usedDynamicInputs []
 usedModules [
  &self:SystemSpawnScript.SystemScriptSource.SystemScriptGraph.NiagaraNodeAssignment_0.SetVariables_#1
  &FN_UnpackVehicleFlags
  /VehicleCosmetics/FX/Niagara/Modules/VC_SetVehicleSpeedAttribs
  &SystemState
 ]
}
emitters {
 NE_VC_DriftTrail_Ribbon_Parent001 {
  emitterUpdateScript [
   &EmitterState()
   &VC_RateQualityHandler(Epic Spawn Rate=30, High Spawn Rate=60, Low Spawn Rate=30, Use Custom Rate?=true) {enabled=false}
   &VC_SpawnRibbon_v1(Number of Ribbons=2, SpawnRate=30)
  ]
  particleSpawnScript [
   &VC_PaintedVariantsOverride(
    Black=<rgba 0.0781 0.0781 0.0781 0.98>
    BurntSienna=<rgba 0.25 0.0733 0.05 1>
    Cobalt=<rgba 0.1 0.1906 5 1>
    Crimson=<rgba 2 0 0.0333 1>
    DefaultColor=<rgba 0.5 0.05 0 0.7>
    ForestGreen=<rgba 0 0.55 0.0278 1>
    Grey=<rgba 0.3 0.3 0.3 0.6>
    Lime=<rgba 0.65 1.3 0 1>
    Orange=<rgba 2 0.6976 0 1>
    Pink=<rgba 1.4 0.1677 0.9519 1>
    Purple=<rgba 0.1667 0 0.4 1>
    R1 Black:LinearColor=<rgba 0 0 0 0.98>
    R1 BurntSienna:LinearColor=<rgba 0.1 0.0309 0.0123 0.9>
    R1 Cobalt:LinearColor=<rgba 0 0.0185 1 0.85>
    R1 Crimson:LinearColor=<rgba 0.15 0.01 0 0.9>
    R1 DefaultColor=<rgba 0 2 1.5 0.2>
    R1 ForestGreen:LinearColor=<rgba 0.01 0.12 0 0.85>
    R1 Grey:LinearColor=<rgba 0.07 0.07 0.07 0.85>
    R1 Lime:LinearColor=<rgba 0.087 0.18 0 0.85>
    R1 Orange:LinearColor=<rgba 0.45 0.1575 0 0.85>
    R1 Pink:LinearColor=<rgba 0.35 0.056 0.2431 0.8>
    R1 Purple:LinearColor=<rgba 0.04 0 0.08 0.88>
    R1 Saffron:LinearColor=<rgba 0.3 0.23 0 0.9>
    R1 SkyBlue:LinearColor=<rgba 0 0.1863 0.28 0.8>
    R1 TitaniumWhite:LinearColor=<rgba 0.75 0.75 0.75 0.7>
    R2 Black:LinearColor=<rgba 0.4 0.4 0.4 0.5>
    R2 BurntSienna:LinearColor=<rgba 0 0 0 1>
    R2 Cobalt:LinearColor=<rgba 0 0 0 1>
    R2 Crimson:LinearColor=<rgba 0 0 0 1>
    R2 DefaultColor:LinearColor=<rgba 0 0 0 1>
    R2 ForestGreen:LinearColor=<rgba 0 0 0 1>
    R2 Grey:LinearColor=<rgba 0 0 0 1>
    R2 Lime:LinearColor=<rgba 0 0 0 1>
    R2 Orange:LinearColor=<rgba 0 0 0 1>
    R2 Pink:LinearColor=<rgba 0 0 0 1>
    R2 Purple:LinearColor=<rgba 0 0 0 1>
    R2 Saffron:LinearColor=<rgba 0 0 0 1>
    R2 SkyBlue:LinearColor=<rgba 0 0 0 1>
    R2 TitaniumWhite:LinearColor=<rgba 0 0 0 1>
    R3 Black:LinearColor=<rgba 0.7 0.7 0.7 0.85>
    R3 BurntSienna:LinearColor=<rgba 0 0 0 1>
    R3 Cobalt:LinearColor=<rgba 0 0 0 1>
    R3 Crimson:LinearColor=<rgba 0 0 0 1>
    R3 DefaultColor:LinearColor=<rgba 0 0 0 1>
    R3 ForestGreen:LinearColor=<rgba 0 0 0 1>
    R3 Grey:LinearColor=<rgba 0 0 0 1>
    R3 Lime:LinearColor=<rgba 0 0 0 1>
    R3 Orange:LinearColor=<rgba 0 0 0 1>
    R3 Pink:LinearColor=<rgba 0 0 0 1>
    R3 Purple:LinearColor=<rgba 0 0 0 1>
    R3 Saffron:LinearColor=<rgba 0 0 0 1>
    R3 SkyBlue:LinearColor=<rgba 0 0 0 1>
    R3 TitaniumWhite:LinearColor=<rgba 0 0 0 1>
    Saffron=<rgba 1.3 1.027 0.13 1>
    SkyBlue=<rgba 0.2191 0.8718 1.2 1>
    TitaniumWhite=<rgba 0.1 0.1 0.1 1>
    Use R1 Painted Overrides=true
    Use R2 Painted Overrides=true
    Use R3 Painted Overrides=true
   )
   &VC_InitializeRibbon_v1(
    Color=$Particles.Color
    Custom Offset=true
    Enable Skeletal Position=true
    Lifetime=1
    Local Vector Offset=&AddVectorToPosition(Position=<-65 0 -30>, Vector=Linked() {linkedVariable {name=User.offset, type=NiagaraPosition}})
    Offset Amount=1
    Ribbon Width=15
   )
   &AddVelocity(Velocity=<-1000 0 0>) {module=AddVelocity001}
  ]
  particleUpdateScript [
   &FN_DisableTireFX(Mask Index Attribute:NiagaraInt32=$Particles.SpawnGroup, Mask Index FL=2, Mask Index FR=3, Mask Index RL=0, Mask Index RR=1)
   &ParticleState()
   &InheritSourceMovement(ApplyForce=false, ApplyPosition=true, ApplyPositionFalloff=false)
   &SolveForcesAndVelocity(Clamp Velocity=false, Limit Acceleration=false)
   /SetVariables(
    Particles.BlackColor:LinearColor=&MultiplyLinearColor(
     Linear Color=$Particles.R2 Color
     Scale Factor=&MakeVector4FromVectorAndFloat(
      "Vector (RGB)"=<1 1 1>
      "Float (Alpha)"=&FloatFromCurve(
       FloatCurve=DataInterface() {propertyValues {Curve {keys [
           <at 0 1 RCIM_Cubic>
           {arriveTangent=-2.2455, interpMode=RCIM_Cubic, leaveTangent=-2.2455, tangentMode=RCTM_User, time=0.2, value=0.1}
           {arriveTangent=1.2121, interpMode=RCIM_Cubic, leaveTangent=1.2121, tangentMode=RCTM_User, time=0.4, value=0.1}
           {arriveTangent=-0.347, interpMode=RCIM_Cubic, leaveTangent=-0.347, time=0.5, value=1}
           <at 0.6882 0 RCIM_Cubic>
          ]}}}
       Scale Curve=&MakeCustomFloatFromBool(Bool=$User.bIsPainted, True Float=2, False Float=0.7)
      )
     )
    )
    Particles.BlueColor:LinearColor=&VC_DynamicScaleColor(
     Color to Scale:LinearColor=$Particles.R1 Color
     RGB Scale:Vector3f=<1 1 1>
     Alpha Scale:NiagaraFloat=&FloatFromCurve(FloatCurve=DataInterface() {propertyValues {Curve {keys [<at 0.609 1>,<at 0.9639 0>]}}})
    )
    Particles.FrameColor:LinearColor=&VC_DynamicScaleColor(
     Color to Scale:LinearColor=$Particles.R3 Color
     RGB Scale:Vector3f=<1 1 1>
     Alpha Scale:NiagaraFloat=&FloatFromCurve(FloatCurve=DataInterface() {propertyValues {Curve {defaultValue=1, keys [{tangentMode=RCTM_User, time=0.88, value=1}, {tangentMode=RCTM_User, time=1, value=0}]}}})
    )
    Particles.Witdh=24
   ) {isSetParameters=true}
   &VC_UpdateRibbon_v1(Enable Skeletal Position=true)
   &DynamicMaterialParameters(
    Index 1 Param 1=&FloatFromCurve(FloatCurve=DataInterface() {propertyValues {Curve {keys [
         {arriveTangent=-4.6625, interpMode=RCIM_Cubic, leaveTangent=-4.6625, tangentMode=RCTM_User, time=0, value=-1.5}
         {arriveTangent=1.4798, interpMode=RCIM_Cubic, leaveTangent=1.4798, tangentMode=RCTM_User, time=0.95, value=1}
        ]}}})
    Module.Index1Float0Write=true
    Module.Index1Float1Write=true
    Module.Index1Float2Write=true
    Module.Index1Float3Write=true
    Write Parameter Index 0=::NewEnumerator0
    Write Parameter Index 1=::NewEnumerator1
   )
   &ScaleColor(
    Scale Alpha=&FloatFromCurve(FloatCurve=DataInterface() {propertyValues {Curve {keys [<at 0.6131 1>,<at 0.9639 0>]}}})
    ScaleA=true
    ScaleRGB=true
   )
  ]
  rendererClasses [&NiagaraRibbonRendererProperties]
  renderers [
   {rendererClass=&NiagaraRibbonRendererProperties}
   {rendererClass=&NiagaraRibbonRendererProperties, rendererIndex=1}
   {rendererClass=&NiagaraRibbonRendererProperties, rendererIndex=2}
   {rendererClass=&NiagaraRibbonRendererProperties, rendererIndex=3}
   {rendererClass=&NiagaraRibbonRendererProperties, rendererIndex=4}
  ]
 }
 NE_VC_DriftTrail_Sprite_Parent002 {
  emitterUpdateScript [
   &EmitterState(Inactive Response=::NewEnumerator1, Life Cycle Mode=::NewEnumerator1, Loop Duration=0.1, UseLoopDelay=false)
   &SpawnBurst_Instantaneous(Spawn Count=6, Use Spawn Probability=false, UseLoopCountLimit=false)
   &SpawnRate(SpawnRate=20, Use Spawn Probability=false) {enabled=false}
  ]
  particleSpawnScript [
   &VC_PaintedVariantsOverride(
    Black=<rgba 1 1 1 0.5>
    BurntSienna=<rgba 0 0 0 1>
    Cobalt=<rgba 0 0 0 1>
    Crimson=<rgba 0 0 0 1>
    DefaultColor=<rgba 0 0 0 1>
    ForestGreen=<rgba 0 0 0 1>
    Grey=<rgba 0 0 0 1>
    Lime=<rgba 0 0 0 1>
    Orange=<rgba 0 0 0 1>
    Pink=<rgba 0 0 0 1>
    Purple=<rgba 0 0 0 1>
    R1 Black:LinearColor=&SelectLinearColorFromArray(
     Array Sampling Mode=::NewEnumerator0
     Direct Array Index=?
     Color Selection Array=DataInterface() {propertyValues {ColorData [<rgba 0.05 0.05 0.05 1>, <rgba 0 0 0 1>]}}
     Random Seed=0
    )
    R1 BurntSienna:LinearColor=&SelectLinearColorFromArray(
     Array Sampling Mode=::NewEnumerator0
     Direct Array Index=?
     Color Selection Array=DataInterface() {propertyValues {ColorData [<rgba 0.16 0.0591 0.032 1>, <rgba 0.1 0.0309 0.0123 1>]}}
     Random Seed=0
    )
    R1 Cobalt:LinearColor=&SelectLinearColorFromArray(
     Array Sampling Mode=::NewEnumerator0
     Direct Array Index=?
     Color Selection Array=DataInterface() {propertyValues {ColorData [<rgba 0.2 0.3867 1 1>, <rgba 0 0.0185 1 1>]}}
     Random Seed=0
    )
    R1 Crimson:LinearColor=&SelectLinearColorFromArray(
     Array Sampling Mode=::NewEnumerator0
     Direct Array Index=?
     Color Selection Array=DataInterface() {propertyValues {ColorData [<rgba 0.5 0 0 1>, <rgba 0.1927 0.0128 0 1>]}}
     Random Seed=0
    )
    R1 DefaultColor=&SelectLinearColorFromArray(
     Array Sampling Mode=::NewEnumerator0
     Direct Array Index=?
     Color Selection Array=DataInterface() {propertyValues {ColorData [<rgba 0 0.72 0.54 0.7>, <rgba 0.375 0.72 0.6338 0.7>, <rgba 1 0 0 0.75>]}}
     Random Seed=0
    )
    R1 ForestGreen:LinearColor=&SelectLinearColorFromArray(
     Array Sampling Mode=::NewEnumerator0
     Direct Array Index=?
     Color Selection Array=DataInterface() {propertyValues {ColorData [<rgba 0 0.55 0.0278 1>, <rgba 0 0.2 0.0101 1>]}}
     Random Seed=0
    )
    R1 Grey:LinearColor=&SelectLinearColorFromArray(
     Array Sampling Mode=::NewEnumerator0
     Direct Array Index=?
     Color Selection Array=DataInterface() {propertyValues {ColorData [<rgba 0.2 0.2 0.2 1>, <rgba 0.085 0.085 0.085 1>]}}
     Random Seed=0
    )
    R1 Lime:LinearColor=&SelectLinearColorFromArray(
     Array Sampling Mode=::NewEnumerator0
     Direct Array Index=?
     Color Selection Array=DataInterface() {propertyValues {ColorData [<rgba 0.5 1 0 1>, <rgba 0.125 0.25 0 1>]}}
     Random Seed=0
    )
    R1 Orange:LinearColor=&SelectLinearColorFromArray(
     Array Sampling Mode=::NewEnumerator0
     Direct Array Index=?
     Color Selection Array=DataInterface() {propertyValues {ColorData [<rgba 2 0.6976 0 1>, <rgba 0.5 0.1744 0 1>]}}
     Random Seed=0
    )
    R1 Pink:LinearColor=&SelectLinearColorFromArray(
     Array Sampling Mode=::NewEnumerator0
     Direct Array Index=?
     Color Selection Array=DataInterface() {propertyValues {ColorData [<rgba 1 0.1198 0.6799 1>, <rgba 0.35 0.056 0.2431 1>]}}
     Random Seed=0
    )
    R1 Purple:LinearColor=&SelectLinearColorFromArray(
     Array Sampling Mode=::NewEnumerator0
     Direct Array Index=?
     Color Selection Array=DataInterface() {propertyValues {ColorData [<rgba 0.0903 0 0.22 1>, <rgba 0.0328 0 0.08 1>]}}
     Random Seed=0
    )
    R1 Saffron:LinearColor=&SelectLinearColorFromArray(
     Array Sampling Mode=::NewEnumerator0
     Direct Array Index=?
     Color Selection Array=DataInterface() {propertyValues {ColorData [<rgba 0.8 0.632 0.08 1>, <rgba 0.3 0.23 0 1>]}}
     Random Seed=0
    )
    R1 SkyBlue:LinearColor=&SelectLinearColorFromArray(
     Array Sampling Mode=::NewEnumerator0
     Direct Array Index=?
     Color Selection Array=DataInterface() {propertyValues {ColorData [<rgba 0 0.6654 1 1>, <rgba 0 0.2528 0.38 1>]}}
     Random Seed=0
    )
    R1 TitaniumWhite:LinearColor=&SelectLinearColorFromArray(
     Array Sampling Mode=::NewEnumerator0
     Direct Array Index=?
     Color Selection Array=DataInterface() {propertyValues {ColorData [<rgba 0.75 0.75 0.75 1>, <rgba 0.4 0.4 0.4 1>]}}
     Random Seed=0
    )
    Saffron=<rgba 0 0 0 1>
    SkyBlue=<rgba 0 0 0 1>
    TitaniumWhite=<rgba 0 0 0 1>
    Use R1 Painted Overrides=true
   )
   &InitializeParticle(
    Color Mode=::NewEnumerator0
    Sprite Rotation Mode=::NewEnumerator2
    Sprite Size Mode=::NewEnumerator3
    Uniform Sprite Size=14
    UsePositionOffset=false
    Write Material Random=true
   )
   &PartitionParticles()
   &SkeletalMeshLocation(
    Bone / Socket Index=&Add_Integer(A=$Particles.PartitionParticles.Partition, B=4)
    Bone / Socket Sampling Mode=::NewEnumerator5
    Interpolate Mesh Sampling=false
    MaskSpawnGroup=false
    Mesh Sampling Type=::NewEnumerator6
    Sampled Position Offset=&AddVectorToPosition(Position=$User.offset, Vector=<-70 0 -25>)
    Skeletal Mesh=$System.SkeletalMesh
   )
   &AddVelocity(Velocity=&VC_Mask_Vectors(
     FrontEnd Value:Vector3f=<-1500 0 0>
     blsFrontEndPreview:NiagaraBool=false
     SpeedMin Value:Vector3f=<-800 0 0>
     SpeedMax Value:Vector3f=<-1 0 0>
     Speed Alpha:NiagaraFloat=$System.Speed_Alpha
    )) {enabled=false}
   &ShapeLocation(MaskSpawnGroup=false, Override Seed=false, Sphere Radius=16, UseOwnerScale=true)
  ]
  particleUpdateScript [
   &FN_DisableTireFX(
    Mask Index Attribute:NiagaraInt32=$Particles.PartitionParticles.Partition
    Mask Index FL=2
    Mask Index FR=3
    Mask Index RL=0
    Mask Index RR=1
   )
   &ParticleState()
   &SolveForcesAndVelocity(Clamp Velocity=false, Limit Acceleration=false)
   /SetVariables(
    Particles.Offset=34
    Particles.r1 scale:LinearColor=&MultiplyLinearColor(
     Linear Color=$Particles.Initial.R1 Color
     Scale Factor=&MakeVector4FromVectorAndFloat(
      "Vector (RGB)"=<1 1 1>
      "Float (Alpha)"=&FloatFromCurve(FloatCurve=DataInterface() {propertyValues {Curve {
          defaultValue=0
          keys [
           {interpMode=RCIM_Cubic, tangentMode=RCTM_User, time=0.5, value=1}
           {arriveTangentWeight=0.0042, interpMode=RCIM_Cubic, leaveTangentWeight=0.4672, tangentMode=RCTM_User, time=0.5, value=0.9998}
           {interpMode=RCIM_Cubic, tangentMode=RCTM_User, time=1, value=0}
          ]
         }}})
     )
    )
   ) {isSetParameters=true}
   &ScaleColor(
    Scale Alpha=&FloatFromCurve(FloatCurve=DataInterface() {propertyValues {Curve {keys [<at 0.8 1 RCIM_Cubic>,<at 1 0 RCIM_Cubic>]}}})
    ScaleA=true
    ScaleRGB=true
   )
   &CameraOffset(Camera Offset Amount=35)
  ]
  rendererClasses [&NiagaraSpriteRendererProperties]
  renderers [
   {rendererClass=&NiagaraSpriteRendererProperties}
   {rendererClass=&NiagaraSpriteRendererProperties, rendererIndex=1}
  ]
 }
 NE_VC_DriftTrail_WheelWrap_Parent001 {
  emitterUpdateScript [
   &EmitterState(
    Inactive Response=::NewEnumerator1
    Life Cycle Mode=::NewEnumerator1
    Loop Behavior=::NewEnumerator1
    Loop Duration Mode=::NewEnumerator1
    UseLoopDelay=false
   )
   &SpawnBurst_Instantaneous(Spawn Count=2, Use Spawn Probability=false, UseLoopCountLimit=false)
  ]
  particleSpawnScript [
   &VC_PaintedVariantsOverride(
    Black=<rgba 0 0 0 1>
    BurntSienna=<rgba 0.14 0.0432 0.0173 1>
    Cobalt=<rgba 0.04 0.0578 1 1>
    Crimson=<rgba 0.1927 0.0128 0 1>
    DefaultColor=<rgba 0.325 2.5 2.1375 0.25>
    ForestGreen=<rgba 0 0.18 0.0091 1>
    Grey=<rgba 0.085 0.085 0.085 1>
    Lime=<rgba 0.15 0.3 0 1>
    Orange=<rgba 0.8 0.2895 0.016 1>
    Pink=<rgba 0.6 0.096 0.4167 1>
    Purple=<rgba 0.0616 0 0.15 1>
    R1 Black:LinearColor=<rgba 0.2 0.2 0.2 1>
    R1 BurntSienna:LinearColor=<rgba 0.3 0.1633 0.1266 1>
    R1 Cobalt:LinearColor=<rgba 0.32 0.4621 8 1>
    R1 Crimson:LinearColor=<rgba 3 0 0.0813 1>
    R1 DefaultColor=<rgba 3 0.12 0.12 0.3>
    R1 ForestGreen:LinearColor=<rgba 0 0.55 0.0278 1>
    R1 Grey:LinearColor=<rgba 0.5 0.5 0.5 1>
    R1 Lime:LinearColor=<rgba 1 2 0 1>
    R1 Orange:LinearColor=<rgba 2 0.6976 0 1>
    R1 Pink:LinearColor=<rgba 3 0.45 2.0727 1>
    R1 Purple:LinearColor=<rgba 0.5335 0 1.3 1>
    R1 Saffron:LinearColor=<rgba 2.5 1.975 0.25 1>
    R1 SkyBlue:LinearColor=<rgba 0.3652 1.453 2 1>
    R1 TitaniumWhite:LinearColor=<rgba 0.1 0.1 0.1 1>
    R2 Black:LinearColor=<rgba 1 1 1 0.75>
    R2 BurntSienna:LinearColor=<rgba 0 0 0 1>
    R2 Cobalt:LinearColor=<rgba 0 0 0 1>
    R2 Crimson:LinearColor=<rgba 0 0 0 1>
    R2 DefaultColor:LinearColor=<rgba 0 0 0 1>
    R2 ForestGreen:LinearColor=<rgba 0 0 0 1>
    R2 Grey:LinearColor=<rgba 0 0 0 1>
    R2 Lime:LinearColor=<rgba 0 0 0 1>
    R2 Orange:LinearColor=<rgba 0 0 0 1>
    R2 Pink:LinearColor=<rgba 0 0 0 1>
    R2 Purple:LinearColor=<rgba 0 0 0 1>
    R2 Saffron:LinearColor=<rgba 0 0 0 1>
    R2 SkyBlue:LinearColor=<rgba 0 0 0 1>
    R2 TitaniumWhite:LinearColor=<rgba 0 0 0 1>
    Saffron=<rgba 0.5 0.3833 0 1>
    SkyBlue=<rgba 0 0.2662 0.4 1>
    Use R1 Painted Overrides=true
    Use R2 Painted Overrides=true
   )
   &InitializeParticle(
    Color=$Particles.Color
    Mesh Scale Mode=::NewEnumerator4
    Sprite Rotation Mode=::NewEnumerator2
    Sprite Size Mode=::NewEnumerator3
    UsePositionOffset=false
    Write Material Random=true
   )
   &PartitionParticles()
   &SkeletalMeshLocation(
    Bone / Socket Index=&Add_Integer(A=$Particles.PartitionParticles.Partition, B=4)
    Bone / Socket Sampling Mode=::NewEnumerator5
    Interpolate Mesh Sampling=false
    MaskSpawnGroup=false
    Mesh Sampling Type=::NewEnumerator6
    Skeletal Mesh=$System.SkeletalMesh
   )
  ]
  particleUpdateScript [
   &FN_DisableTireFX(
    Mask Index Attribute:NiagaraInt32=$Particles.PartitionParticles.Partition
    Mask Index FL=2
    Mask Index FR=3
    Mask Index RL=0
    Mask Index RR=1
   )
   &VC_WheelWrapScale(Wheel Index Attribute:NiagaraInt32=&Add_Integer(A=$Particles.PartitionParticles.Partition, B=2))
   &ParticleState(Kill Particles When Lifetime Has Elapsed=false)
   &UpdateMeshOrientation(Rotation Coordinate Space=::NewEnumerator2, Rotation Vector=<0 1 0>) {enabled=false}
   &SkeletalMeshLocation(
    Bone / Socket Index=&Add_Integer(A=$Particles.PartitionParticles.Partition, B=4)
    Bone / Socket Sampling Mode=::NewEnumerator5
    MaskSpawnGroup=false
    Mesh Sampling Type=::NewEnumerator6
    Skeletal Mesh=$System.SkeletalMesh
   ) {module=SkeletalMeshLocation001}
   &SolveForcesAndVelocity(Clamp Velocity=false, Limit Acceleration=false)
   /SetVariables(
    Particles.Linear Color=&VC_DynamicScaleColor(Color to Scale:LinearColor=<rgba 0 0 0 1>, RGB Scale:Vector3f=<1 1 1>, Alpha Scale:NiagaraFloat=1)
    Particles.Orange Color:LinearColor=&VC_DynamicScaleColor(Color to Scale:LinearColor=$Particles.R1 Color, RGB Scale:Vector3f=<1 1 1>, Alpha Scale:NiagaraFloat=1)
   ) {isSetParameters=true}
  ]
  rendererClasses [&NiagaraMeshRendererProperties]
  renderers [
   {rendererClass=&NiagaraMeshRendererProperties, rendererIndex=0}
   {rendererClass=&NiagaraMeshRendererProperties, rendererIndex=1}
   {rendererClass=&NiagaraMeshRendererProperties, rendererIndex=2}
  ]
 }
}
summary {
 systemName=NS_DriftTrail_Gaudy_MP_
 emitters [
  {emitterName=NE_VC_DriftTrail_Ribbon_Parent001, rendererClasses [&NiagaraRibbonRendererProperties]}
  {emitterName=NE_VC_DriftTrail_Sprite_Parent002, rendererClasses [&NiagaraSpriteRendererProperties]}
  {emitterName=NE_VC_DriftTrail_WheelWrap_Parent001, rendererClasses [&NiagaraMeshRendererProperties]}
 ]
 userVariables [
  User.bIsFrontendPreview {type=NiagaraBool}
  User.bIsPainted {type=NiagaraBool}
  User.ColorVariantID {type=NiagaraInt32}
  User.offset {type=NiagaraPosition, description="", defaultValue {}}
  User.PaintedColor {type=LinearColor, description="", defaultValue {struct=LinearColor, value=<rgba 1 1 1 1>}}
  User.PaintedScalar {type=NiagaraFloat, defaultValue=1, description=""}
  User.VehicleFlags {type=NiagaraInt32, defaultValue=806289408}
  User.Wheel_Radius_Array {
   type=NiagaraDataInterfaceArrayFloat
   defaultValue {
    struct=NiagaraExt_VariableValue_DataInterface
    value {
     dataInterface=&self:NiagaraDataInterfaceArrayFloat_1
     dataInterfaceClass=&NiagaraDataInterfaceArrayFloat
    }
   }
  }
  User.Wheel_Width_Array {
   type=NiagaraDataInterfaceArrayFloat
   defaultValue {
    struct=NiagaraExt_VariableValue_DataInterface
    value {
     dataInterface=&self:NiagaraDataInterfaceArrayFloat_0
     dataInterfaceClass=&NiagaraDataInterfaceArrayFloat
    }
   }
  }
  User.WheelTouchingArray {
   type=NiagaraDataInterfaceArrayFloat
   defaultValue {
    struct=NiagaraExt_VariableValue_DataInterface
    value {
     dataInterface=&self:NiagaraDataInterfaceArrayFloat_2
     dataInterfaceClass=&NiagaraDataInterfaceArrayFloat
    }
   }
  }
 ]
}
